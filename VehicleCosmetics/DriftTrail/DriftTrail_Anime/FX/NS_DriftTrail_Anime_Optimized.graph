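# Niagara drift-trail vehicle cosmetic in an anime style, an optimized variant emitting from the two rear wheels.
# Two ribbon emitters draw layered colored streaks with waveform-driven width and material parameters, alongside three mesh and sprite particle emitters whose spawn or velocity inputs use speed-based masks.
# A looping wheel-wrap mesh follows the wheel sockets and alone applies the user painted color.
# Every emitter runs FN_DisableTireFX to mask effects per wheel.
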NiagaraSystem /VehicleCosmetics/DriftTrail/DriftTrail_Anime/FX/NS_DriftTrail_Anime_Optimized
dependencies {
 usedDataInterfaces [
  &NiagaraDataInterfaceVectorCurve
  &NiagaraDataInterfaceCurve
  &NiagaraDataInterfaceColorCurve
  &NiagaraDataInterfaceSkeletalMesh
 ]
 usedDynamicInputs []
 usedModules [
  &self:SystemSpawnScript.SystemScriptSource.SystemScriptGraph.NiagaraNodeAssignment_0.SetVariables_#1
  &SystemState
  &FN_UnpackVehicleFlags
  /Game/Effects/Niagara/Modules/Vehicle/FN_UpdateVehicleAttribs
  /Game/Effects/Niagara/Modules/Vehicle/FN_VehicleFlags_Cheatcodes
 ]
}
emitters {
 NE_Trail {
  emitterUpdateScript [
   &EmitterState(Life Cycle Mode=::NewEnumerator1, Loop Duration=1, UseLoopDelay=false)
   &VC_RateQualityHandler(Epic Spawn Rate=90, High Spawn Rate=60, Low Spawn Rate=30, Medium Spawn Rate=45, Use Custom Rate?=true)
   &VC_SpawnRibbon_v1(Number of Ribbons=2, SpawnRate=35)
  ]
  particleSpawnScript [
   &VC_InitializeRibbon_v1(Custom Offset=true, Enable Skeletal Position=true, Local Vector Offset=<60 -6 -35>, Offset Amount=1, Ribbon Width=7)
   &AddVelocity(Velocity=&MakeCustomVectorFromBool(Bool=$User.bIsFrontendPreview, True Vector=<-4000 0 0>, False Vector=<-4000 0 0>)) {module=AddVelocity001}
   &PartitionParticles()
  ]
  particleUpdateScript [
   &FN_DisableTireFX(Mask Index Attribute:NiagaraInt32=$Particles.SpawnGroup, Mask Index FL=2, Mask Index FR=3, Mask Index RL=0, Mask Index RR=1)
   &ParticleState()
   &InheritSourceMovement(
    Applied Position Scale=&VectorFromCurve(VectorCurve=DataInterface() {propertyValues {
       XCurve {keys [<at 0 1 RCIM_Cubic>, {interpMode=RCIM_Cubic, leaveTangent=-1.655, time=0.7, value=0}]}
       YCurve {keys [<at 0 1 RCIM_Cubic>, {interpMode=RCIM_Cubic, leaveTangent=-1.655, time=0.7, value=0}]}
       ZCurve {keys [<at 0 1 RCIM_Cubic>, {interpMode=RCIM_Cubic, leaveTangent=-1.655, time=0.7, value=0}]}
      }})
    ApplyForce=false
    ApplyPosition=true
    ApplyPositionFalloff=false
   )
   &SolveForcesAndVelocity(Clamp Velocity=false, Limit Acceleration=false)
   &VC_UpdateRibbon_v1(Enable Skeletal Position=true)
   &ScaleRibbonWidth(Ribbon Width Scale=&FloatFromCurve(
     FloatCurve=DataInterface() {propertyValues {Curve {keys [<at 0 1.48>,<at 1 0>]}}}
     Scale Curve=&Waveform(Global Amplitude Scale=3, "Waveform [1]"=::NewEnumerator0, "[1] Frequency"=5, "[1] Amplitude Min/Max"=<0.9 1>)
    ))
   &ScaleColor(
    Linear Color Curve=DataInterface() {propertyValues {
      AlphaCurve {keys [<at 0.1393 0>,<at 0.1574 0.5>,<at 1 0>]}
      BlueCurve {keys [<at 0.2287 0.682>]}
      GreenCurve {keys [<at 0.2287 0.25>]}
      RedCurve {keys [<at 0.2287 1>]}
     }}
    Scale Mode=::NewEnumerator2
   )
   &DynamicMaterialParameters(
    Index 1 Param 2=&Waveform(Global Amplitude Scale=7, "Waveform [1]"=::NewEnumerator10, "[1] Phase Scale"=15, "[1] Frequency"=1.5, "[1] Amplitude Min/Max"=<0 1>)
    Index 1 Param 3=&Add_Float(
     A=&Waveform(Global Amplitude Scale=10, "Waveform [1]"=::NewEnumerator10, "[1] Frequency"=30, "[1] Amplitude Min/Max"=<0 1>)
     B=&Waveform(Global Amplitude Scale=7, "Waveform [1]"=::NewEnumerator10, "[1] Frequency"=1.5, "[1] Amplitude Min/Max"=<0 1>)
    )
    Index 3 Param 1=10
    Index 3 Param 2=10
    Module.Index1Float0Write=false
    Module.Index1Float1Write=true
    Module.Index1Float2Write=true
    Module.Index1Float3Write=true
    Module.Index3Float0Write=false
    Module.Index3Float1Write=false
    Module.Index3Float2Write=false
    Module.Index3Float3Write=false
    Write Parameter Index 0=::NewEnumerator0
    Write Parameter Index 1=::NewEnumerator1
    Write Parameter Index 3=::NewEnumerator1
   )
  ]
  rendererClasses [&NiagaraRibbonRendererProperties]
  renderers [{rendererClass=&NiagaraRibbonRendererProperties}]
 }
 NE_VC_DriftTrail_Ribbon_Parent {
  emitterUpdateScript [
   &EmitterState()
   &VC_SpawnRibbon_v1(Number of Ribbons=2, SpawnRate=60)
  ]
  particleSpawnScript [
   &VC_PaintedVariantsOverride() {enabled=false}
   &VC_InitializeRibbon_v1(
    Color=<rgba 0 0 0 0>
    Custom Offset=true
    Enable Skeletal Position=true
    Lifetime=1.25
    Local Vector Offset=<0 0 -30>
    Offset Amount=1
    Ribbon Width=30
   )
   &AddVelocity(Velocity=<-1500 0 0>) {module=AddVelocity001}
   /SetVariables(Particles.CoreBeam_Color:LinearColor=<rgba 1 1 1 0>) {isSetParameters=true}
  ]
  particleUpdateScript [
   &FN_DisableTireFX(Mask Index Attribute:NiagaraInt32=$Particles.SpawnGroup, Mask Index FL=2, Mask Index FR=3, Mask Index RL=0, Mask Index RR=1)
   &ParticleState()
   &SolveForcesAndVelocity(Clamp Velocity=false, Limit Acceleration=false)
   &VC_UpdateRibbon_v1(Enable Skeletal Position=true)
   &DynamicMaterialParameters(
    Index 0 Param 1=1
    Index 0 Param 2=&FloatFromCurve(FloatCurve=DataInterface() {propertyValues {Curve {
        defaultValue=0.3
        keys [
         {arriveTangent=1.7504, interpMode=RCIM_Cubic, leaveTangent=1.7504, tangentMode=RCTM_User, time=0, value=0.75}
         {interpMode=RCIM_Cubic, tangentMode=RCTM_User, time=0.5, value=1}
        ]
       }}})
    Index 0 Param 3=&FloatFromCurve(
     FloatCurve=DataInterface() {propertyValues {Curve {defaultValue=1, keys [{tangentMode=RCTM_User, time=0.05, value=0}, {tangentMode=RCTM_User, time=0.2, value=2}]}}}
     Scale Curve=10
    )
    Index 0 Param 4=&FloatFromCurve(
     FloatCurve=DataInterface() {propertyValues {Curve {defaultValue=0.7, keys [{tangentMode=RCTM_User, time=0, value=0.5}, {tangentMode=RCTM_User, time=1, value=1.25}]}}}
     Scale Curve=0.65
    )
    Module.Index0Float0Write=true
    Module.Index0Float1Write=true
    Module.Index0Float2Write=true
    Module.Index0Float3Write=true
   )
   /SetVariables(Particles.RibbonFacing=<1 0 -1>, Particles.RibbonFacing2:Vector3f=<0 0 -1>) {isSetParameters=true}
   /SetVariables(
    Particles.Color=&ColorFromCurve(ColorCurve=DataInterface() {propertyValues {
       AlphaCurve {keys [<at 0.0383 0>,<at 0.0737 1.75>,<at 1 0>]}
       BlueCurve {keys [<at 0.2084 1>,<at 0.5022 1>]}
       GreenCurve {keys [<at 0.2084 0.075>,<at 0.5022 0.6424>]}
       RedCurve {keys [<at 0.2084 1>,<at 0.5022 0.2>]}
      }})
    Particles.Ribbon 2 Dynamic:Vector4f=&Multiply_Vector4(A=$Particles.DynamicMaterialParameter, B=<1 1 -1 1>)
   ) {isSetParameters=true}
   /SetVariables(
    Particles.CoreBeam_Color:LinearColor=&ColorFromCurve(ColorCurve=DataInterface() {propertyValues {
       AlphaCurve {keys [<at 0.1494 0>,<at 0.2869 0.2>,<at 0.4625 0.1953>,<at 1 0>]}
       BlueCurve {keys [<at 0 2>]}
       GreenCurve {keys [<at 0 0>]}
       RedCurve {keys [<at 0 0.25>]}
      }})
    Particles.CoreBeam_Width:NiagaraFloat=&FloatFromCurve(
     FloatCurve=DataInterface() {propertyValues {Curve {
        defaultValue=1
        keys [
         {arriveTangent=2.0076, interpMode=RCIM_Cubic, leaveTangent=2.0076, tangentMode=RCTM_User, time=0, value=0}
         {arriveTangent=0.1247, interpMode=RCIM_Cubic, leaveTangent=0.1247, tangentMode=RCTM_User, time=0.05, value=1}
         {arriveTangent=0.1225, interpMode=RCIM_Cubic, leaveTangent=0.1225, tangentMode=RCTM_User, time=1, value=0}
        ]
       }}}
     Scale Curve=13
    )
   ) {isSetParameters=true}
  ]
  rendererClasses [&NiagaraRibbonRendererProperties]
  renderers [
   {rendererClass=&NiagaraRibbonRendererProperties}
   {rendererClass=&NiagaraRibbonRendererProperties, rendererIndex=1}
   {rendererClass=&NiagaraRibbonRendererProperties, rendererIndex=2}
  ]
 }
 NE_VC_DriftTrail_Sprite_Parent {
  emitterUpdateScript [
   &EmitterState()
   &SpawnRate(
    SpawnRate=/VehicleCosmetics/FX/Niagara/Modules/VC_Mask_Floats(
     FrontEnd Value:NiagaraFloat=20
     blsFrontEndPreview:NiagaraBool=false
     SpeedMin Value:NiagaraFloat=15
     SpeedMax Value:NiagaraFloat=20
     Speed Alpha:NiagaraFloat=$System.Speed_Alpha
    )
    Use Spawn Probability=false
   )
  ]
  particleSpawnScript [
   &VC_PaintedVariantsOverride() {enabled=false}
   &InitializeParticle(
    Color Mode=::NewEnumerator0
    Lifetime=0.35
    Mesh Scale Mode=::NewEnumerator3
    Mesh Uniform Scale=0.5
    Sprite Rotation Mode=::NewEnumerator2
    Sprite Size Mode=::NewEnumerator3
    UsePositionOffset=false
    Write Material Random=true
   )
   &PartitionParticles()
   &SkeletalMeshLocation(
    Bone / Socket Index=&Add_Integer(A=$Particles.PartitionParticles.Partition, B=4)
    Bone / Socket Sampling Mode=::NewEnumerator5
    Interpolate Mesh Sampling=false
    MaskSpawnGroup=false
    Mesh Sampling Type=::NewEnumerator6
    Sampled Position Offset=<-75 0 -25>
    Skeletal Mesh=$System.SkeletalMesh
   )
   &AddVelocity(Velocity=&VC_Mask_Vectors(
     FrontEnd Value:Vector3f=<-3500 0 0>
     blsFrontEndPreview:NiagaraBool=false
     SpeedMin Value:Vector3f=<-3500 0 0>
     SpeedMax Value:Vector3f=<-1 0 0>
     Speed Alpha:NiagaraFloat=$System.Speed_Alpha
    ))
   &Color(Color=<rgba 0.15 0.1796 1 10>) {module=Color001}
  ]
  particleUpdateScript [
   &FN_DisableTireFX(
    Mask Index Attribute:NiagaraInt32=$Particles.PartitionParticles.Partition
    Mask Index FL=2
    Mask Index FR=3
    Mask Index RL=0
    Mask Index RR=1
   )
   &ParticleState()
   &ScaleMeshSize(Scale Factor=&VectorFromFloat(Value=&FloatFromCurve(FloatCurve=DataInterface() {propertyValues {Curve {
         defaultValue=1
         keys [
          {arriveTangent=9.0725, interpMode=RCIM_Cubic, leaveTangent=9.0725, tangentMode=RCTM_User, time=0, value=0.5}
          {arriveTangent=-0.7835, interpMode=RCIM_Cubic, leaveTangent=-0.7835, tangentMode=RCTM_User, time=0.1408, value=0.9847}
          {arriveTangent=-1.6166, interpMode=RCIM_Cubic, leaveTangent=-1.6166, tangentMode=RCTM_User, time=1, value=0}
         ]
        }}}))) {module=ScaleMeshSize001}
   &SolveForcesAndVelocity(Clamp Velocity=false, Limit Acceleration=false)
   &ScaleColor(
    Scale Alpha=&FloatFromCurve(FloatCurve=DataInterface() {propertyValues {Curve {
        defaultValue=1
        keys [
         {arriveTangent=-0.1012, interpMode=RCIM_Cubic, leaveTangent=-0.1012, tangentMode=RCTM_User, time=0.8, value=1}
         {arriveTangent=-5.8612, interpMode=RCIM_Cubic, leaveTangent=-5.8612, tangentMode=RCTM_User, time=1, value=0}
        ]
       }}})
    ScaleA=true
    ScaleRGB=false
   )
   &DynamicMaterialParameters(
    Index 2 Param 3=1
    Index 2 Param 4=&RandomRangeFloat(Minimum=0.005, Maximum=0.02)
    Module.Index2Float0Write=false
    Module.Index2Float1Write=false
    Module.Index2Float2Write=true
    Module.Index2Float3Write=true
    Write Parameter Index 0=::NewEnumerator0
    Write Parameter Index 2=::NewEnumerator1
   )
  ]
  rendererClasses [&NiagaraMeshRendererProperties, &NiagaraSpriteRendererProperties]
  renderers [
   {rendererClass=&NiagaraMeshRendererProperties, rendererIndex=0}
   {rendererClass=&NiagaraSpriteRendererProperties, rendererIndex=1}
  ]
 }
 NE_VC_DriftTrail_Sprite_Parent001 {
  emitterUpdateScript [
   &EmitterState()
   &SpawnRate(
    SpawnRate=&VC_Mask_Floats(
     FrontEnd Value:NiagaraFloat=10
     blsFrontEndPreview:NiagaraBool=false
     SpeedMin Value:NiagaraFloat=10
     SpeedMax Value:NiagaraFloat=10
     Speed Alpha:NiagaraFloat=$System.Speed_Alpha
    )
    Use Spawn Probability=false
   )
  ]
  particleSpawnScript [
   &VC_PaintedVariantsOverride() {enabled=false}
   &InitializeParticle(
    Color Mode=::NewEnumerator0
    Lifetime=0.5
    Mesh Scale Mode=::NewEnumerator3
    Mesh Uniform Scale=0.25
    Sprite Rotation Mode=::NewEnumerator2
    Sprite Size Mode=::NewEnumerator1
    Uniform Sprite Size Max=5
    Uniform Sprite Size Min=3
    UsePositionOffset=false
    Write Material Random=true
   )
   &PartitionParticles()
   &SkeletalMeshLocation(
    Bone / Socket Index=&Add_Integer(A=$Particles.PartitionParticles.Partition, B=4)
    Bone / Socket Sampling Mode=::NewEnumerator5
    Interpolate Mesh Sampling=false
    MaskSpawnGroup=false
    Mesh Sampling Type=::NewEnumerator6
    Sampled Position Offset=<-75 0 -30>
    Skeletal Mesh=$System.SkeletalMesh
   )
   &AddVelocity(Velocity=&VC_Mask_Vectors(
     FrontEnd Value:Vector3f=<-4500 0 0>
     blsFrontEndPreview:NiagaraBool=false
     SpeedMin Value:Vector3f=<-4500 0 0>
     SpeedMax Value:Vector3f=<-1 0 0>
     Speed Alpha:NiagaraFloat=$System.Speed_Alpha
    ))
   &Color(Color=<rgba 0.5454 0.3438 2 1>) {module=Color001}
   &ShapeLocation(MaskSpawnGroup=false, Override Seed=false, Sphere Radius=25, UseOwnerScale=true) {module=ShapeLocation001}
   /SetVariables(
    Particles.Color1:LinearColor=<rgba 1 1 1 10>
    Particles.Color2=<rgba 0 0 0 10>
    Particles.Size1:Vector2f=&Multiply_Vector2D(A=$Particles.SpriteSize, B=<7 7>)
    Particles.Size2:Vector2f=&Multiply_Vector2D(A=$Particles.SpriteSize, B=<15 15>)
   ) {isSetParameters=true}
  ]
  particleUpdateScript [
   &FN_DisableTireFX(
    Mask Index Attribute:NiagaraInt32=$Particles.PartitionParticles.Partition
    Mask Index FL=2
    Mask Index FR=3
    Mask Index RL=0
    Mask Index RR=1
   )
   &InheritSourceMovement(
    Applied Position Scale=&VectorFromCurve(
     VectorCurve=DataInterface() {propertyValues {
       XCurve {keys [<at 0 1 RCIM_Cubic>, {interpMode=RCIM_Cubic, leaveTangent=-1.655, time=0.7, value=0}]}
       YCurve {keys [<at 0 1 RCIM_Cubic>, {interpMode=RCIM_Cubic, leaveTangent=-1.655, time=0.7, value=0}]}
       ZCurve {keys [<at 0 1 RCIM_Cubic>, {interpMode=RCIM_Cubic, leaveTangent=-1.655, time=0.7, value=0}]}
      }}
     Scale Curve=<0.25 0.25 0.25>
    )
    ApplyForce=false
    ApplyPosition=true
    ApplyPositionFalloff=false
   )
   &ParticleState()
   &ScaleSpriteSize(Uniform Curve Sprite Scale=DataInterface() {propertyValues {Curve {
       defaultValue=0
       keys [
        {arriveTangent=4.1272, interpMode=RCIM_Cubic, leaveTangent=4.1272, tangentMode=RCTM_User, time=0, value=0}
        {arriveTangent=0.0224, interpMode=RCIM_Cubic, leaveTangent=0.0224, tangentMode=RCTM_User, time=1, value=1}
       ]
      }}})
   &ScaleColor(
    Scale Alpha=&FloatFromCurve(FloatCurve=DataInterface() {propertyValues {Curve {
        defaultValue=1
        keys [
         {arriveTangent=-1.6639, interpMode=RCIM_Cubic, leaveTangent=-1.6639, tangentMode=RCTM_User, time=0, value=1}
         {interpMode=RCIM_Cubic, tangentMode=RCTM_User, time=1, value=0}
        ]
       }}})
    ScaleA=true
    ScaleRGB=false
   )
   &DynamicMaterialParameters(
    Index 0 Param 1=&FloatFromCurve(FloatCurve=DataInterface() {propertyValues {Curve {
        defaultValue=1
        keys [{interpMode=RCIM_Cubic, tangentMode=RCTM_User, time=0.0073, value=0.5}, {interpMode=RCIM_Cubic, tangentMode=RCTM_User, time=1, value=2}]
       }}})
    Index 0 Param 2=&FloatFromCurve(FloatCurve=DataInterface() {propertyValues {Curve {
        defaultValue=1
        keys [{interpMode=RCIM_Cubic, tangentMode=RCTM_User, time=0, value=1}, {interpMode=RCIM_Cubic, tangentMode=RCTM_User, time=1, value=6}]
       }}})
    Module.Index0Float0Write=true
    Module.Index0Float1Write=true
    Module.Index0Float2Write=false
    Module.Index0Float3Write=false
   )
   &SolveForcesAndVelocity(Clamp Velocity=false, Limit Acceleration=false)
  ]
  rendererClasses [&NiagaraSpriteRendererProperties]
  renderers [
   {rendererClass=&NiagaraSpriteRendererProperties}
   {rendererClass=&NiagaraSpriteRendererProperties, rendererIndex=1}
   {rendererClass=&NiagaraSpriteRendererProperties, rendererIndex=2}
  ]
 }
 NE_VC_DriftTrail_Sprite_Parent002 {
  emitterUpdateScript [
   &EmitterState()
   &SpawnRate(
    SpawnRate=&VC_Mask_Floats(
     FrontEnd Value:NiagaraFloat=100
     blsFrontEndPreview:NiagaraBool=false
     SpeedMin Value:NiagaraFloat=100
     SpeedMax Value:NiagaraFloat=100
     Speed Alpha:NiagaraFloat=$System.Speed_Alpha
    )
    Use Spawn Probability=false
   )
  ]
  particleSpawnScript [
   &VC_PaintedVariantsOverride() {enabled=false}
   &InitializeParticle(
    Color=<rgba 90 90 300 1>
    Lifetime=0.1
    Mesh Scale Mode=::NewEnumerator3
    Mesh Uniform Scale=0.25
    Position Offset=<-20 0 0>
    Sprite Rotation Mode=::NewEnumerator2
    Sprite Size Mode=::NewEnumerator1
    Uniform Sprite Size Max=4.5
    Uniform Sprite Size Min=1.5
    UsePositionOffset=true
    Write Material Random=true
   )
   &PartitionParticles()
   &SkeletalMeshLocation(
    Bone / Socket Index=&Add_Integer(A=$Particles.PartitionParticles.Partition, B=4)
    Bone / Socket Sampling Mode=::NewEnumerator5
    Interpolate Mesh Sampling=false
    MaskSpawnGroup=false
    Mesh Sampling Type=::NewEnumerator6
    Sampled Position Offset=<-30 0 -30>
    Skeletal Mesh=$System.SkeletalMesh
   )
   &AddVelocity(
    Cone Angle=240
    Cone Axis=<-1 0 0>
    Module.UseConeFalloff=false
    Module.UseVelDistribution=false
    Override Seed=false
    Velocity Mode=::NewEnumerator2
    Velocity Speed=1250
   )
   &ApplyInitialForces()
  ]
  particleUpdateScript [
   &FN_DisableTireFX(
    Mask Index Attribute:NiagaraInt32=$Particles.PartitionParticles.Partition
    Mask Index FL=2
    Mask Index FR=3
    Mask Index RL=0
    Mask Index RR=1
   )
   &InheritSourceMovement(
    Applied Position Scale=&VectorFromCurve(
     VectorCurve=DataInterface() {propertyValues {
       XCurve {keys [<at 0 1 RCIM_Cubic>, {interpMode=RCIM_Cubic, leaveTangent=-1.655, time=0.7, value=0}]}
       YCurve {keys [<at 0 1 RCIM_Cubic>, {interpMode=RCIM_Cubic, leaveTangent=-1.655, time=0.7, value=0}]}
       ZCurve {keys [<at 0 1 RCIM_Cubic>, {interpMode=RCIM_Cubic, leaveTangent=-1.655, time=0.7, value=0}]}
      }}
     Scale Curve=<0.25 0.25 0.25>
    )
    ApplyForce=false
    ApplyPosition=true
    ApplyPositionFalloff=false
   )
   &ParticleState()
   &ScaleSpriteSize(Uniform Curve Sprite Scale=DataInterface() {propertyValues {Curve {
       defaultValue=0
       keys [
        {arriveTangent=4.1272, interpMode=RCIM_Cubic, leaveTangent=4.1272, tangentMode=RCTM_User, time=0, value=0}
        {arriveTangent=0.0224, interpMode=RCIM_Cubic, leaveTangent=0.0224, tangentMode=RCTM_User, time=0.5, value=1}
       ]
      }}})
   &Drag(Drag=10, Use Linear Drag=true, Use Rotational Drag=true)
   &SolveForcesAndVelocity(Clamp Velocity=false, Limit Acceleration=false)
   &ScaleSpriteSizeBySpeed(Max Scale Factor=<1 3>, Min Scale Factor=<1 0.5>)
  ]
  rendererClasses [&NiagaraSpriteRendererProperties]
  renderers [
   {rendererClass=&NiagaraSpriteRendererProperties}
   {rendererClass=&NiagaraSpriteRendererProperties, rendererIndex=1}
  ]
 }
 NE_VC_DriftTrail_WheelWrap_R001 {
  emitterUpdateScript [
   &EmitterState(
    Inactive Response=::NewEnumerator1
    Life Cycle Mode=::NewEnumerator1
    Loop Behavior=::NewEnumerator1
    Loop Duration Mode=::NewEnumerator1
    UseLoopDelay=false
   )
   &SpawnBurst_Instantaneous(Spawn Count=2, Use Spawn Probability=false, UseLoopCountLimit=false)
  ]
  particleSpawnScript [
   &VC_PaintedVariantsOverride() {enabled=false}
   &InitializeParticle(
    Color=&MakeCustomLinearColorFromBool(Bool=$User.bIsPainted, True Color=$User.PaintedColor, False Color=<rgba 1 1 1 1>)
    Mesh Scale Mode=::NewEnumerator4
    Sprite Rotation Mode=::NewEnumerator2
    Sprite Size Mode=::NewEnumerator3
    UsePositionOffset=false
    Write Material Random=true
   )
   &PartitionParticles()
   &SkeletalMeshLocation(
    Bone / Socket Index=&Add_Integer(A=$Particles.PartitionParticles.Partition, B=4)
    Bone / Socket Sampling Mode=::NewEnumerator5
    Interpolate Mesh Sampling=false
    MaskSpawnGroup=false
    Mesh Sampling Type=::NewEnumerator6
    Orientation Sampling=::NewEnumerator1
    Sampled Position Offset Coordinate Space:ENiagaraSkelMeshTransforms=::NewEnumerator2
    Skeletal Mesh=$System.SkeletalMesh
   )
  ]
  particleUpdateScript [
   &FN_DisableTireFX(
    Mask Index Attribute:NiagaraInt32=$Particles.PartitionParticles.Partition
    Mask Index FL=2
    Mask Index FR=3
    Mask Index RL=0
    Mask Index RR=1
   )
   &VC_WheelWrapScale(Wheel Index Attribute:NiagaraInt32=&Add_Integer(A=$Particles.PartitionParticles.Partition, B=2))
   &DynamicMaterialParameters(
    Index 0 Param 1=$User.RibbonCoreIntensity
    Index 0 Param 2=$User.RibbonMidIntensity
    Index 0 Param 3=$User.RibbonOutIntensity
    Index 0 Param 4=&MakeCustomFloatFromBool(Bool=$User.bIsPainted, True Float=1, False Float=0)
    Module.Index0Param0Write=true
    Module.Index0Param1Write=true
    Module.Index0Param2Write=true
    Module.Index0Param3Write=true
   )
   &ParticleState(Kill Particles When Lifetime Has Elapsed=false, Loop Particles Lifetime=true)
   &UpdateMeshOrientation(Rotation Coordinate Space=::NewEnumerator2, Rotation Vector=<0 1 0>) {enabled=false}
   &SkeletalMeshLocation(
    Bone / Socket Index=&Add_Integer(A=$Particles.PartitionParticles.Partition, B=4)
    Bone / Socket Sampling Mode=::NewEnumerator5
    MaskSpawnGroup=false
    Mesh Sampling Type=::NewEnumerator6
    Skeletal Mesh=$System.SkeletalMesh
   ) {module=SkeletalMeshLocation001}
   &SolveForcesAndVelocity(Clamp Velocity=false, Limit Acceleration=false)
  ]
  rendererClasses [&NiagaraMeshRendererProperties]
  renderers [{rendererClass=&NiagaraMeshRendererProperties, rendererIndex=0}]
 }
}
summary {
 systemName=NS_DriftTrail_Anime_Optimized
 emitters [
  {emitterName=NE_Trail, rendererClasses [&NiagaraRibbonRendererProperties]}
  {emitterName=NE_VC_DriftTrail_WheelWrap_R001, rendererClasses [&NiagaraMeshRendererProperties]}
  {emitterName=NE_VC_DriftTrail_Ribbon_Parent, rendererClasses [&NiagaraRibbonRendererProperties]}
  {
   emitterName=NE_VC_DriftTrail_Sprite_Parent
   rendererClasses [&NiagaraMeshRendererProperties, &NiagaraSpriteRendererProperties]
  }
  {emitterName=NE_VC_DriftTrail_Sprite_Parent001, rendererClasses [&NiagaraSpriteRendererProperties]}
  {emitterName=NE_VC_DriftTrail_Sprite_Parent002, rendererClasses [&NiagaraSpriteRendererProperties]}
 ]
 userVariables [
  User.bIsFrontendPreview {type=NiagaraBool}
  User.bIsPainted {type=NiagaraBool}
  User.ColorVariantID {type=NiagaraInt32}
  User.CoreIntensity {type=NiagaraFloat, defaultValue=1, description=""}
  User.Float_Particles_Size1 {type=NiagaraFloat, defaultValue=1, description=""}
  User.InColorIntensity {type=NiagaraFloat, defaultValue=1, description=""}
  User.offset {type=Vector3f, description="offset from wheels(set through cosmetic template)", defaultValue {struct=Vector3f, value=<0 0 0>}}
  User.OutColorIntensity {type=NiagaraFloat, defaultValue=1, description=""}
  User.PaintedColor {type=LinearColor, description="", defaultValue {struct=LinearColor, value=<rgba 1 1 1 1>}}
  User.PaintedColor_2 {type=LinearColor, description="", defaultValue {struct=LinearColor, value=<rgba 1 1 1 1>}}
  User.PaintedScalar {type=NiagaraFloat, defaultValue=1, description=""}
  User.RibbonCoreIntensity {type=NiagaraFloat, defaultValue=10, description=""}
  User.RibbonMidIntensity {type=NiagaraFloat, defaultValue=5, description=""}
  User.RibbonOutIntensity {type=NiagaraFloat, defaultValue=2, description=""}
  User.RingsIntensityB {type=NiagaraFloat, defaultValue=5, description=""}
  User.RingsIntensityG {type=NiagaraFloat, defaultValue=5, description=""}
  User.RingsIntensityR {type=NiagaraFloat, defaultValue=5, description=""}
  User.RingsOpacity {type=NiagaraFloat, defaultValue=1, description=""}
  User.SingleBoostPos {type=NiagaraPosition, description="", defaultValue {}}
  User.SingleColorIntensity {type=NiagaraFloat, defaultValue=1, description=""}
  User.VehicleFlags {type=NiagaraInt32, defaultValue=806289408}
  User.Wheel_Radius_Array {
   type=NiagaraDataInterfaceArrayFloat
   defaultValue {
    struct=NiagaraExt_VariableValue_DataInterface
    value {
     dataInterface=&self:NiagaraDataInterfaceArrayFloat_0
     dataInterfaceClass=&NiagaraDataInterfaceArrayFloat
    }
   }
  }
  User.Wheel_Width_Array {
   type=NiagaraDataInterfaceArrayFloat
   defaultValue {
    struct=NiagaraExt_VariableValue_DataInterface
    value {
     dataInterface=&self:NiagaraDataInterfaceArrayFloat_1
     dataInterfaceClass=&NiagaraDataInterfaceArrayFloat
    }
   }
  }
  User.WheelSize {type=NiagaraFloat, defaultValue=0, description=""}
  User.WheelTouchingArray {
   type=NiagaraDataInterfaceArrayFloat
   defaultValue {
    struct=NiagaraExt_VariableValue_DataInterface
    value {
     dataInterface=&self:NiagaraDataInterfaceArrayFloat_2
     dataInterfaceClass=&NiagaraDataInterfaceArrayFloat
    }
   }
  }
 ]
}
